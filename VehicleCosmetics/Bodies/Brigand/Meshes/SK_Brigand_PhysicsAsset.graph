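# Physics asset for the Brigand vehicle body skeletal mesh in the vehicle cosmetics content.
# It holds a single body on the frame bone, approximated by one large rotated capsule offset from the bone origin.
# It defines no constraints, so the asset contains no articulated or jointed parts.
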PhysicsAsset /VehicleCosmetics/Bodies/Brigand/Meshes/SK_Brigand_PhysicsAsset
bodies {frame [{shapeType=Capsule, length=294.6029, radius=117.8156, center=<1.7825 -85.1278 -4.5225>, rotation=<rot -1.8246 -90 -16.6973>}]}
constraints []
